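# Physics asset for the Baba Yaga female cosmetic body's dangling extras: rope, mushrooms, necklace, grass tufts and front leg armor.
# Mostly capsule bodies hang from pelvis, with the necklace on spine_05; spine_05 and thigh_r are locked to pelvis.
# Locked base and belt bodies anchor most pieces, which get narrow swing limits and mostly locked twist, including a four-link rope chain.
# The leg armor chain instead has free swing and twist.
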PhysicsAsset /BRCosmetics/Characters/Player/Female/Medium/Bodies/F_MED_BabaYaga/Meshes/F_MED_BabaYaga_ExtrasPhysics
bodies {
 dyn_backBottomGrassB [{shapeType=Capsule, length=5.1031, radius=1.8, shapeName=dyn_backBottomGrassB_capsule, center=<0 0 0.5>}]
 dyn_backBottomGrassC [{shapeType=Capsule, length=4.0981, radius=1.8, shapeName=dyn_backBottomGrassC_capsule, center=<0 0 -0.5>}]
 dyn_backTopGrassA [{shapeType=Capsule, length=7, radius=1.6, shapeName=dyn_backTopGrassA_capsule, center=<0 0 1>}]
 dyn_backTopGrassB [{shapeType=Capsule, length=7, radius=1.6, shapeName=dyn_backTopGrassB_capsule, center=<0 0 -1>}]
 dyn_backTopGrassMain_Base [{shapeType=Capsule, length=0.505, radius=0.505, rotation=<rot 0 -90 0>}]
 dyn_frontGrass [{shapeType=Capsule, length=7.5797, radius=1, center=<1.5 0 0>, rotation=<rot 90 0 0>}]
 dyn_frontGrass_Base [{shapeType=Capsule, length=0.505, radius=0.505, rotation=<rot 0 -90 0>}]
 dyn_legArmorFrontMain [{
   shapeType=Capsule
   length=22.7866
   radius=18.75
   shapeName=dyn_legArmorFrontMain_capsule
   center=<6.2339 16.225 0.1287>
   rotation=<rot -86.2022 -0.0001 2.1559>
  }]
 dyn_legArmorFront_02_l [
  {
   shapeType=Capsule
   length=12.75
   radius=19.75
   shapeName=dyn_legArmorFront_02_l_capsule
   center=<6.8339 18.5053 0.7705>
   rotation=<rot -83.6241 -0.0001 0.0001>
  }
  {
   shapeType=Capsule
   length=20.75
   radius=9
   shapeName=dyn_legArmorFront_02_l_capsule
   center=<9.3631 2.0658 -12.9612>
   rotation=<rot 0 0 -120.1824>
  }
 ]
 dyn_mushroomBeltFrontMain_01 [{shapeType=Capsule, length=0.2, radius=0.2, center=<5.2229 -0.7439 0>, rotation=<rot 90 0 0>}]
 dyn_mushroomBeltFrontMain_01_Base [{shapeType=Capsule, length=0.505, radius=0.505, rotation=<rot 0 -90 0>}]
 dyn_mushroomBeltFrontMain_02 [{shapeType=Capsule, length=0.3, radius=0.3, rotation=<rot 90 0 0>}]
 dyn_mushroomBeltSideMain_01 [{shapeType=Capsule, length=0.505, radius=0.505, rotation=<rot 0 -90 0>}]
 dyn_mushroomBeltSideMain_01_Base [{shapeType=Capsule, length=0.505, radius=0.505, rotation=<rot 0 -90 0>}]
 dyn_mushroom_A [{
   shapeType=Box
   extentX=14.4858
   extentY=3.4927
   extentZ=13.7601
   shapeName=dyn_mushroom_A_box
   center=<5.0436 -1.4516 1.9835>
   rotation=<rot 79.5172 105.9208 100.5161>
  }]
 dyn_mushroom_C [{shapeType=Capsule, length=4.9021, radius=3.4626, center=<2.4127 -2 0>, rotation=<rot 90 0 0>}]
 dyn_mushroom_D [{shapeType=Capsule, length=4.8849, radius=3.8661, center=<3 -3 0>, rotation=<rot 90 0 -9.6761>}]
 dyn_mushroom_E [{shapeType=Capsule, length=4.9081, radius=3.5301, center=<4.1714 0.0001 0>, rotation=<rot 90 0 0>}]
 dyn_necklaceMain [{shapeType=Capsule, length=6.2065, radius=1.5239, center=<4.6084 0.0001 0>, rotation=<rot 90 0 0>}]
 dyn_necklaceMain_Base [{shapeType=Capsule, length=0.505, radius=0.505, rotation=<rot 0 -90 0>}]
 dyn_rope_01 [{shapeType=Capsule, length=4.5707, radius=2.2185, shapeName=dyn_rope_01_capsule, center=<3 -1.5 0>, rotation=<rot 90 0 0>}]
 dyn_rope_01_Base [{shapeType=Capsule, length=0.505, radius=0.505, rotation=<rot 0 -90 0>}]
 dyn_rope_02 [{shapeType=Capsule, length=4.3544, radius=1.8266, center=<3.033 -1.5 0.0001>, rotation=<rot 90 0 0>}]
 dyn_rope_03 [{shapeType=Capsule, length=5.7399, radius=1.7208, center=<3.2196 -1.5 0.0001>, rotation=<rot 90 0 0>}]
 dyn_rope_04 [{shapeType=Capsule, length=2.4391, radius=2.5346, center=<3.2986 -1.5 -0.0001>, rotation=<rot 90 0 0>}]
 pelvis [{shapeType=Capsule, length=0.0005, radius=15.0467, center=<-4.8245 -3.4809 -0.1474>, rotation=<rot 3.3058 -90 -86.7993>}]
 spine_05 [{shapeType=Capsule, length=9.577, radius=11.5008, center=<-5.4632 0 0>}]
 thigh_r [{shapeType=Capsule, length=23.9627, radius=9.6748, center=<13.7702 -0.0007 0>, rotation=<rot 85.0822 176.2395 -180>}]
}
constraints [
 {
  bone1Name=dyn_rope_01_Base
  bone2Name=pelvis
  swing1LimitDegrees=45
  swing1Motion=Locked
  swing2LimitDegrees=45
  swing2Motion=Locked
  twistLimitDegrees=45
  twistMotion=Locked
 }
 {bone1Name=dyn_rope_02, bone2Name=dyn_rope_01, swing1LimitDegrees=10, swing2LimitDegrees=10, twistLimitDegrees=8, twistMotion=Locked}
 {bone1Name=dyn_rope_03, bone2Name=dyn_rope_02, swing1LimitDegrees=10, swing2LimitDegrees=10, twistLimitDegrees=8, twistMotion=Locked}
 {bone1Name=dyn_rope_04, bone2Name=dyn_rope_03, swing1LimitDegrees=10, swing2LimitDegrees=10, twistLimitDegrees=8, twistMotion=Locked}
 {
  bone1Name=dyn_mushroomBeltSideMain_01_Base
  bone2Name=pelvis
  swing1LimitDegrees=45
  swing1Motion=Locked
  swing2LimitDegrees=45
  swing2Motion=Locked
  twistLimitDegrees=45
  twistMotion=Locked
 }
 {
  bone1Name=dyn_mushroomBeltSideMain_01
  bone2Name=dyn_mushroomBeltSideMain_01_Base
  swing1LimitDegrees=45
  swing1Motion=Locked
  swing2LimitDegrees=45
  swing2Motion=Locked
  twistLimitDegrees=45
  twistMotion=Locked
 }
 {bone1Name=dyn_mushroom_E, bone2Name=dyn_mushroomBeltSideMain_01, swing1LimitDegrees=16, swing2LimitDegrees=10, twistLimitDegrees=6}
 {bone1Name=dyn_mushroom_D, bone2Name=dyn_mushroomBeltSideMain_01, swing1LimitDegrees=16, swing2LimitDegrees=10, twistLimitDegrees=6}
 {
  bone1Name=dyn_mushroomBeltFrontMain_01_Base
  bone2Name=pelvis
  swing1LimitDegrees=45
  swing1Motion=Locked
  swing2LimitDegrees=45
  swing2Motion=Locked
  twistLimitDegrees=45
  twistMotion=Locked
 }
 {
  bone1Name=dyn_mushroomBeltFrontMain_01
  bone2Name=dyn_mushroomBeltFrontMain_01_Base
  swing1LimitDegrees=45
  swing1Motion=Locked
  swing2LimitDegrees=45
  swing2Motion=Locked
  twistLimitDegrees=45
  twistMotion=Locked
 }
 {
  bone1Name=dyn_mushroom_C
  bone2Name=dyn_mushroomBeltFrontMain_02
  swing1LimitDegrees=16
  swing2LimitDegrees=10
  twistLimitDegrees=6
  twistMotion=Locked
 }
 {
  bone1Name=thigh_r
  bone2Name=pelvis
  swing1LimitDegrees=45
  swing1Motion=Locked
  swing2LimitDegrees=45
  swing2Motion=Locked
  twistLimitDegrees=45
  twistMotion=Locked
 }
 {
  bone1Name=dyn_mushroomBeltFrontMain_02
  bone2Name=dyn_mushroomBeltFrontMain_01
  swing1LimitDegrees=45
  swing1Motion=Locked
  swing2LimitDegrees=45
  swing2Motion=Locked
  twistLimitDegrees=45
  twistMotion=Locked
 }
 {
  bone1Name=dyn_necklaceMain_Base
  bone2Name=spine_05
  swing1LimitDegrees=45
  swing1Motion=Locked
  swing2LimitDegrees=45
  swing2Motion=Locked
  twistLimitDegrees=45
  twistMotion=Locked
 }
 {
  bone1Name=dyn_necklaceMain
  bone2Name=dyn_necklaceMain_Base
  swing1LimitDegrees=8
  swing2LimitDegrees=20
  twistLimitDegrees=45
  twistMotion=Locked
 }
 {
  bone1Name=spine_05
  bone2Name=pelvis
  swing1LimitDegrees=45
  swing1Motion=Locked
  swing2LimitDegrees=45
  swing2Motion=Locked
  twistLimitDegrees=45
  twistMotion=Locked
 }
 {
  bone1Name=dyn_frontGrass_Base
  bone2Name=pelvis
  swing1LimitDegrees=45
  swing1Motion=Locked
  swing2LimitDegrees=45
  swing2Motion=Locked
  twistLimitDegrees=45
  twistMotion=Locked
 }
 {
  bone1Name=dyn_frontGrass
  bone2Name=dyn_frontGrass_Base
  swing1LimitDegrees=11
  swing2LimitDegrees=11
  twistLimitDegrees=6
  twistMotion=Locked
 }
 {
  bone1Name=dyn_backTopGrassMain_Base
  bone2Name=pelvis
  swing1LimitDegrees=45
  swing1Motion=Locked
  swing2LimitDegrees=45
  swing2Motion=Locked
  twistLimitDegrees=45
  twistMotion=Locked
 }
 {
  bone1Name=dyn_mushroom_A
  bone2Name=dyn_mushroomBeltFrontMain_02
  swing1LimitDegrees=16
  swing2LimitDegrees=10
  twistLimitDegrees=45
  twistMotion=Locked
 }
 {
  bone1Name=dyn_backTopGrassA
  bone2Name=dyn_backTopGrassMain_Base
  swing1LimitDegrees=15
  swing2LimitDegrees=15
  twistLimitDegrees=8
  twistMotion=Locked
 }
 {
  bone1Name=dyn_backTopGrassB
  bone2Name=dyn_backTopGrassMain_Base
  swing1LimitDegrees=15
  swing2LimitDegrees=15
  twistLimitDegrees=8
  twistMotion=Locked
 }
 {bone1Name=dyn_backBottomGrassB, bone2Name=pelvis, swing1LimitDegrees=8, swing2LimitDegrees=8, twistLimitDegrees=45, twistMotion=Locked}
 {bone1Name=dyn_backBottomGrassC, bone2Name=pelvis, swing1LimitDegrees=8, swing2LimitDegrees=8, twistLimitDegrees=45, twistMotion=Locked}
 {
  bone1Name=dyn_legArmorFront_02_l
  bone2Name=dyn_legArmorFrontMain
  swing1LimitDegrees=45
  swing1Motion=Free
  swing2LimitDegrees=45
  swing2Motion=Free
  twistLimitDegrees=45
  twistMotion=Free
 }
 {
  bone1Name=dyn_legArmorFrontMain
  bone2Name=pelvis
  swing1LimitDegrees=45
  swing1Motion=Free
  swing2LimitDegrees=45
  swing2Motion=Free
  twistLimitDegrees=45
  twistMotion=Free
 }
 {bone1Name=dyn_rope_01, bone2Name=dyn_rope_01_Base, swing1LimitDegrees=10, swing2LimitDegrees=10, twistLimitDegrees=10, twistMotion=Locked}
]
