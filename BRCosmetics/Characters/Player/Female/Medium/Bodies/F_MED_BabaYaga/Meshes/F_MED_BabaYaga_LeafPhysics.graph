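# Physics asset for the BabaYaga medium female body's leaf dressing, covering a right-arm leaf and a shoulder armor leaf.
# Capsule bodies sit on the two leaf bones, their parent pieces and the right upper-arm twist bone.
# The arm leaf base has swing and twist locked to the upper arm, and each leaf is jointed to its base piece.
# The shoulder armor pair has no constraint to the arm bodies.
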PhysicsAsset /BRCosmetics/Characters/Player/Female/Medium/Bodies/F_MED_BabaYaga/Meshes/F_MED_BabaYaga_LeafPhysics
bodies {
 dyn_armLeaf_B_Base_r [{shapeType=Capsule, length=0.505, radius=0.505, center=<-0.3472 -0.4528 2.4031>, rotation=<rot 0 -151.9899 -67.8677>}]
 dyn_armLeaf_B_r [{shapeType=Capsule, length=3.2301, radius=2.0965, center=<1.2892 0.0006 0>, rotation=<rot 0 90 90>}]
 dyn_shoulderArmorLeaf_A [{shapeType=Capsule, length=3.3861, radius=2.2687, center=<1.8609 0.0001 0>, rotation=<rot 0 90 90>}]
 dyn_shoulderArmorMain [{shapeType=Capsule, length=1, radius=1, rotation=<rot 90 0 0>}]
 upperarm_twist_01_r [{shapeType=Capsule, length=13.9474, radius=4.0282, center=<0.3955 -0.5152 0.4579>, rotation=<rot 0 -99.039 -90.6231>}]
}
constraints [
 {bone1Name=dyn_armLeaf_B_r, bone2Name=dyn_armLeaf_B_Base_r, swing1LimitDegrees=20, swing2LimitDegrees=8, twistLimitDegrees=8}
 {bone1Name=dyn_shoulderArmorLeaf_A, bone2Name=dyn_shoulderArmorMain, swing1LimitDegrees=20, swing2LimitDegrees=12, twistLimitDegrees=8}
 {
  bone1Name=dyn_armLeaf_B_Base_r
  bone2Name=upperarm_twist_01_r
  swing1LimitDegrees=20
  swing1Motion=Locked
  swing2LimitDegrees=8
  swing2Motion=Locked
  twistLimitDegrees=8
  twistMotion=Locked
 }
]
