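# Physics asset for the F_MED_BabaYaga female medium body that simulates secondary motion of its skirt panels and back armor.
# Pelvis, thigh and calf capsules are joined by fully locked constraints, forming a rigid leg skeleton.
# Box and capsule chains on dyn_ bones hang from it: back skirt, front skirt and back armor from the pelvis, side panels from the thighs.
# Angular limits and selective axis locks restrict each chain's swing and twist, with asymmetric left and right side setups.
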PhysicsAsset /BRCosmetics/Characters/Player/Female/Medium/Bodies/F_MED_BabaYaga/Meshes/F_MED_BabaYaga_Physics
bodies {
 calf_l [{shapeType=Capsule, length=41.5097, radius=7.056, center=<-22.3885 -3.5073 -0.0003>, rotation=<rot 90 0 180>}]
 calf_r [{shapeType=Capsule, length=41.5097, radius=7.056, center=<22.3885 3.5073 0>, rotation=<rot 90 0 0>}]
 dyn_backArmor_01 [{shapeType=Box, extentX=22, extentY=2, extentZ=26.7362, center=<11.4 -6.4 0>}]
 dyn_backArmor_02 [{shapeType=Capsule, length=16, radius=4, shapeName=dyn_backArmor_02_capsule, center=<13.7437 -1 0>, rotation=<rot 0 0.0001 -9.7794>}]
 dyn_skirtBackTopMain_c [{shapeType=Capsule, length=20, radius=6, shapeName=dyn_skirtBackTopMain_c_capsule, center=<2.9723 0 0>, rotation=<rot 0 0 90>}]
 dyn_skirtBackTop_01_c [{
   shapeType=Box
   extentX=4.3443
   extentY=17.8192
   extentZ=18.462
   shapeName=dyn_skirtBackTop_01_c_box
   center=<4.5505 2.1 0>
   rotation=<rot 4.1316 99.4681 5.7379>
  }]
 dyn_skirtBackTop_02_c [{shapeType=Box, extentX=3.8084, extentY=28.3667, extentZ=24.0557, shapeName=dyn_skirtBackTop_02_c_box, center=<10 0 0>, rotation=<rot 5 90 0>}]
 dyn_skirtFrontBottom_01_l [{shapeType=Capsule, length=21.75, radius=5.25, shapeName=dyn_skirtFrontBottom_01_l_capsule, center=<1.5252 -0.0001 0>}]
 dyn_skirtFrontBottom_02_l [{shapeType=Capsule, length=5, radius=6.5, shapeName=dyn_skirtFrontBottom_02_l_capsule, center=<1.5252 -0.0001 0>}]
 dyn_skirtFrontMain [{shapeType=Capsule, length=5, radius=3.75, shapeName=dyn_skirtFrontMain_capsule, center=<1.6217 -0.0001 0>}]
 dyn_skirtFrontMain_Base [{shapeType=Box, extentX=1.01, extentY=1.01, extentZ=1.01, rotation=<rot 0 -90 0>}]
 dyn_skirtSideBack_01_l [{shapeType=Capsule, length=15, radius=6.042, shapeName=dyn_skirtSideBack_01_l_capsule, center=<-8.1829 2.4 0>}]
 dyn_skirtSideBack_01_r [{shapeType=Capsule, length=12, radius=4, shapeName=dyn_skirtSideBack_01_r_capsule, center=<8.1117 -2 -1>, rotation=<rot 90 0 0>}]
 dyn_skirtSideFore_01_l [{shapeType=Box, extentX=3.0286, extentY=12.827, extentZ=12.7238, center=<-6.1953 -0.0002 0>, rotation=<rot 0 90 180>}]
 dyn_skirtSideFore_01_r [{shapeType=Box, extentX=2.0794, extentY=12.827, extentZ=12.7238, center=<6.1953 0.0002 0>, rotation=<rot 0 90 0>}]
 dyn_skirtSideFore_02_l [{shapeType=Box, extentX=6.3722, extentY=12.8202, extentZ=11.1443, center=<-6.9884 0 0>, rotation=<rot 0 90 0>}]
 dyn_skirtSideFore_02_r [{shapeType=Box, extentX=1.5516, extentY=11.3202, extentZ=11.1443, center=<5.2256 0 0>, rotation=<rot 0 90 0>}]
 dyn_skirtSideMid_01_r [{shapeType=Box, extentX=3.3671, extentY=12.3638, extentZ=9.8346, center=<5.915 0 -0.0001>, rotation=<rot 0 90 0>}]
 pelvis [{shapeType=Capsule, length=0.0005, radius=12, center=<-4.8245 -0.0091 -0.348>, rotation=<rot 3.3058 -90 -86.7993>}]
 thigh_l [{shapeType=Capsule, length=28.7656, radius=7.2, center=<-20.544 -2.7695 0>, rotation=<rot 90 0 0>}]
 thigh_r [{shapeType=Capsule, length=28.7656, radius=7.2, center=<20.544 2.7695 0>, rotation=<rot 90 0 0>}]
}
constraints [
 {bone1Name=dyn_skirtSideFore_02_r, bone2Name=dyn_skirtSideFore_01_r, swing1LimitDegrees=20, swing2LimitDegrees=16, twistLimitDegrees=8}
 {bone1Name=dyn_skirtBackTop_02_c, bone2Name=dyn_skirtBackTop_01_c, swing1LimitDegrees=45, swing2LimitDegrees=15, twistLimitDegrees=15}
 {bone1Name=dyn_skirtSideFore_01_r, bone2Name=thigh_r, swing1LimitDegrees=20, swing2LimitDegrees=16, twistLimitDegrees=8}
 {bone1Name=dyn_skirtSideMid_01_r, bone2Name=thigh_r, swing1LimitDegrees=20, swing2LimitDegrees=16, twistLimitDegrees=8}
 {bone1Name=dyn_skirtSideBack_01_r, bone2Name=thigh_r, swing1LimitDegrees=20, swing2LimitDegrees=45, twistLimitDegrees=5}
 {
  bone1Name=thigh_l
  bone2Name=pelvis
  swing1LimitDegrees=45
  swing1Motion=Locked
  swing2LimitDegrees=45
  swing2Motion=Locked
  twistLimitDegrees=45
  twistMotion=Locked
 }
 {
  bone1Name=thigh_r
  bone2Name=pelvis
  swing1LimitDegrees=45
  swing1Motion=Locked
  swing2LimitDegrees=45
  swing2Motion=Locked
  twistLimitDegrees=45
  twistMotion=Locked
 }
 {bone1Name=dyn_skirtSideFore_01_l, bone2Name=thigh_l, swing1LimitDegrees=20, swing2LimitDegrees=16, twistLimitDegrees=8}
 {bone1Name=dyn_skirtSideFore_02_l, bone2Name=dyn_skirtSideFore_01_l, swing1LimitDegrees=20, swing2LimitDegrees=16, twistLimitDegrees=8}
 {
  bone1Name=dyn_skirtSideBack_01_l
  bone2Name=thigh_l
  swing1LimitDegrees=20
  swing2LimitDegrees=16
  swing2Motion=Locked
  twistLimitDegrees=8
  twistMotion=Locked
 }
 {
  bone1Name=calf_r
  bone2Name=thigh_r
  swing1LimitDegrees=45
  swing1Motion=Locked
  swing2LimitDegrees=45
  swing2Motion=Locked
  twistLimitDegrees=45
  twistMotion=Locked
 }
 {
  bone1Name=dyn_skirtBackTopMain_c
  bone2Name=pelvis
  swing1LimitDegrees=10
  swing1Motion=Locked
  swing2LimitDegrees=45
  twistLimitDegrees=10
  twistMotion=Locked
 }
 {
  bone1Name=dyn_skirtFrontMain_Base
  bone2Name=pelvis
  swing1LimitDegrees=45
  swing1Motion=Locked
  swing2LimitDegrees=45
  swing2Motion=Locked
  twistLimitDegrees=45
  twistMotion=Locked
 }
 {
  bone1Name=dyn_skirtFrontMain
  bone2Name=dyn_skirtFrontMain_Base
  swing1LimitDegrees=30
  swing2LimitDegrees=20
  twistLimitDegrees=10
  twistMotion=Locked
 }
 {
  bone1Name=dyn_backArmor_01
  bone2Name=pelvis
  swing1LimitDegrees=10
  swing2LimitDegrees=4
  swing2Motion=Locked
  twistLimitDegrees=45
  twistMotion=Locked
 }
 {
  bone1Name=calf_l
  bone2Name=thigh_l
  swing1LimitDegrees=45
  swing1Motion=Locked
  swing2LimitDegrees=45
  swing2Motion=Locked
  twistLimitDegrees=45
  twistMotion=Locked
 }
 {
  bone1Name=dyn_skirtBackTop_01_c
  bone2Name=dyn_skirtBackTopMain_c
  swing1LimitDegrees=45
  swing2LimitDegrees=5
  swing2Motion=Locked
  twistLimitDegrees=20
  twistMotion=Locked
 }
 {bone1Name=dyn_skirtFrontBottom_02_l, bone2Name=dyn_skirtFrontBottom_01_l, swing1LimitDegrees=45, swing2LimitDegrees=20, twistLimitDegrees=10}
 {bone1Name=dyn_skirtFrontBottom_01_l, bone2Name=dyn_skirtFrontMain, swing1LimitDegrees=45, swing2LimitDegrees=20, twistLimitDegrees=10}
 {bone1Name=dyn_backArmor_02, bone2Name=dyn_backArmor_01, swing1LimitDegrees=10, swing2LimitDegrees=5, twistLimitDegrees=5, twistMotion=Locked}
]
