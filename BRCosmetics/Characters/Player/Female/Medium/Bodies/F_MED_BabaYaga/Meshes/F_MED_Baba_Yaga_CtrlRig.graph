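# Control Rig for the Baba Yaga player body that procedurally adjusts skirt, leg armor, rope, belt and arm-leaf bones to follow the pose.
# A helper bone constrained between both calves aims at the right calf and pelvis, driving front-piece rotation.
# Weighted constraints blend back-side skirt and arm-leaf bases between neighboring body bones.
# Thigh angles relative to pelvis weight additive corrections, with left skirt ones scaled by `is crouching`.
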ControlRigBlueprint /BRCosmetics/Characters/Player/Female/Medium/Bodies/F_MED_BabaYaga/Meshes/F_MED_Baba_Yaga_CtrlRig
graphs [
 RigVMFunctionLibrary {links [], nodes []}
 RigVMModel {nodes [
   RigUnit_BeginExecution = &RigUnit_BeginExecution()
   /Script/ControlRig.RigUnit_TransformConstraint(
    BaseTransform=(Rotation=(X=0,Y=0,Z=0,W=1),Translation=(X=0,Y=0,Z=0),Scale3D=(X=1,Y=1,Z=1))
    BaseTransformSpace=GlobalSpace
    Bone=dyn_kneeAim_l
    Targets="((Transform=(Rotation=(X=0,Y=0,Z=0,W=1),Translation=(X=0,Y=0,Z=0),Scale3D=(X=1,Y=1,Z=1)),Weight=1,bMaintainOffset=True,Filter=(TranslationFilter=(bX=True,bY=True,bZ=True),RotationFilter=(bX=True,bY=True,bZ=True),ScaleFilter=(bX=True,bY=True,bZ=True))),(Tra ...(truncated)"
    bUseInitialTransforms=True
    Targets.0.Transform.Translation=&RigUnit_GetBoneTransform(Bone=calf_l).Transform.Translation
    Targets.1.Transform.Translation=&RigUnit_GetBoneTransform(Bone=calf_r).Transform.Translation
    ExecutePin=@RigUnit_BeginExecution.ExecutePin
   ) {is=struct RigVMUnitNode}
   > /Script/ControlRig.RigUnit_DrawContainerSetTransform(Transform=&RigUnit_GetBoneTransform(Bone=dyn_kneeAim_l).Transform) {is=struct RigVMUnitNode}
   RigVMStructNode_0 = &RigUnit_GetBoneTransform(Bone=calf_r)
   > /Script/ControlRig.RigUnit_AimBone(
    Bone=dyn_kneeAim_l
    DebugSettings=(bEnabled=true,Scale=10,WorldOffset=(Rotation=(X=0,Y=0,Z=0,W=1),Translation=(X=0,Y=0,Z=0),Scale3D=(X=1,Y=1,Z=1)))
    Primary=(Weight=1,Axis=(X=0,Y=0,Z=1),Target=(X=1,Y=0,Z=0),Kind=Location)
    Secondary=(Weight=1,Axis=(X=1,Y=0,Z=-0.01),Target=(X=1,Y=0,Z=0),Kind=Location)
    Weight=1
    bPropagateToChildren=True
    Secondary.Target=&RigUnit_GetBoneTransform(Bone=pelvis).Transform.Translation
    Primary.Target=@RigVMStructNode_0.Transform.Translation
   ) {is=struct RigVMUnitNode}
   n1 = &RigUnit_GetBoneTransform(Bone=dyn_kneeAim_l)
   > &RigUnit_TransformConstraint(
    BaseBone=pelvis
    BaseTransform=(Rotation=(X=0,Y=0,Z=0,W=1),Translation=(X=0,Y=0,Z=0),Scale3D=(X=1,Y=1,Z=1))
    BaseTransformSpace=GlobalSpace
    Bone=dyn_skirtFrontMain_BaseUpper
    Targets=((Transform=(Rotation=(X=0,Y=0,Z=0,W=1),Translation=(X=0,Y=0,Z=0),Scale3D=(X=1,Y=1,Z=1)),Weight=1,bMaintainOffset=True,Filter=(TranslationFilter=(bX=True,bY=True,bZ=True),RotationFilter=(bX=True,bY=True,bZ=True),ScaleFilter=(bX=True,bY=True,bZ=True))))
    bUseInitialTransforms=True
    Targets.0.Transform.Rotation=@n1.Transform.Rotation
    Targets.0.Transform.Translation=&RigUnit_GetBoneTransform(Bone=dyn_skirtFrontMain_BaseUpper).Transform.Translation
   )
   > &RigUnit_TransformConstraint(
    BaseBone=pelvis
    BaseTransform=(Rotation=(X=0,Y=0,Z=0,W=1),Translation=(X=0,Y=0,Z=0),Scale3D=(X=1,Y=1,Z=1))
    BaseTransformSpace=GlobalSpace
    Bone=dyn_legArmorFrontMain_Base
    Targets=((Transform=(Rotation=(X=0,Y=0,Z=0,W=1),Translation=(X=0,Y=0,Z=0),Scale3D=(X=1,Y=1,Z=1)),Weight=1,bMaintainOffset=True,Filter=(TranslationFilter=(bX=True,bY=True,bZ=True),RotationFilter=(bX=True,bY=True,bZ=True),ScaleFilter=(bX=True,bY=True,bZ=True))))
    bUseInitialTransforms=True
    Targets.0.Transform.Rotation=@n1.Transform.Rotation
    Targets.0.Transform.Translation=&RigUnit_GetBoneTransform(Bone=dyn_legArmorFrontMain_Base).Transform.Translation
   )
   > &RigUnit_TransformConstraint(
    BaseBone=pelvis
    BaseTransform=(Rotation=(X=0,Y=0,Z=0,W=1),Translation=(X=0,Y=0,Z=0),Scale3D=(X=1,Y=1,Z=1))
    BaseTransformSpace=GlobalSpace
    Bone=dyn_rope_01_Base
    Targets=((Transform=(Rotation=(X=0,Y=0,Z=0,W=1),Translation=(X=0,Y=0,Z=0),Scale3D=(X=1,Y=1,Z=1)),Weight=1,bMaintainOffset=True,Filter=(TranslationFilter=(bX=True,bY=True,bZ=True),RotationFilter=(bX=True,bY=True,bZ=True),ScaleFilter=(bX=True,bY=True,bZ=True))))
    bUseInitialTransforms=True
    Targets.0.Transform.Rotation=@n1.Transform.Rotation
    Targets.0.Transform.Translation=&RigUnit_GetBoneTransform(Bone=dyn_rope_01_Base).Transform.Translation
   )
   > /Script/ControlRig.RigUnit_AddBoneTransform(
    Bone=dyn_legArmorFrontMain_Base
    Transform=(Rotation=(X=0,Y=0,Z=0,W=1),Translation=(X=0,Y=0,Z=0),Scale3D=(X=1,Y=1,Z=1))
    Weight=1
    bPropagateToChildren=True
    Transform.Rotation=&RigVMFunction_MathQuaternionFromEuler(Euler=(X=9,Y=0,Z=0)).Result
   ) {is=struct RigVMUnitNode}
   n2 = &RigUnit_GetBoneTransform(Bone=pelvis)
   > &RigUnit_TransformConstraint(
    BaseBone=pelvis
    BaseTransform=(Rotation=(X=0,Y=0,Z=0,W=1),Translation=(X=0,Y=0,Z=0),Scale3D=(X=1,Y=1,Z=1))
    BaseTransformSpace=GlobalSpace
    Bone=dyn_skirtSideBackTop_01_r_Base
    Targets="((Transform=(Rotation=(X=0,Y=0,Z=0,W=1),Translation=(X=0,Y=0,Z=0),Scale3D=(X=1,Y=1,Z=1)),Weight=0.8,bMaintainOffset=True,Filter=(TranslationFilter=(bX=True,bY=True,bZ=True),RotationFilter=(bX=True,bY=True,bZ=True),ScaleFilter=(bX=True,bY=True,bZ=True))),(T ...(truncated)"
    bUseInitialTransforms=True
    Targets.0.Transform=&RigUnit_GetBoneTransform(Bone=thigh_r).Transform
    Targets.1.Transform=@n2.Transform
   )
   > &RigUnit_TransformConstraint(
    BaseBone=None
    BaseTransform=(Rotation=(X=0,Y=0,Z=0,W=1),Translation=(X=0,Y=0,Z=0),Scale3D=(X=1,Y=1,Z=1))
    BaseTransformSpace=GlobalSpace
    Bone=dyn_skirtSideBackTop_01_l_Base
    Targets="((Transform=(Rotation=(X=0,Y=0,Z=0,W=1),Translation=(X=0,Y=0,Z=0),Scale3D=(X=1,Y=1,Z=1)),Weight=0.8,bMaintainOffset=True,Filter=(TranslationFilter=(bX=True,bY=True,bZ=True),RotationFilter=(bX=True,bY=True,bZ=True),ScaleFilter=(bX=True,bY=True,bZ=True))),(T ...(truncated)"
    bUseInitialTransforms=True
    Targets.0.Transform=&RigUnit_GetBoneTransform(Bone=thigh_l).Transform
    Targets.1.Transform=@n2.Transform
   )
   n3 = &RigUnit_GetRelativeBoneTransform(Bone=thigh_r, Space=pelvis)
   n4 = &RigVMFunction_MathFloatRemap(
    SourceMaximum=80
    SourceMinimum=172
    TargetMinimum=0
    bClamp=true
    Value=/Script/RigVM.RigVMFunction_MathFloatAbs(Value=&RigVMFunction_MathQuaternionToEuler(RotationOrder=ZYX, Value=@n3.Transform.Rotation).Result.Z) {is=struct RigVMUnitNode, template="Absolute::Execute(in Value,out Result)"}.Result
   )
   > /Script/ControlRig.RigUnit_ModifyBoneTransforms(
    BoneToModify=((Bone="dyn_legArmor_01_r",Transform=(Rotation=(X=0,Y=0,Z=0,W=1),Translation=(X=0,Y=0,Z=0),Scale3D=(X=1,Y=1,Z=1))))
    Mode=AdditiveLocal
    WeightMaximum=1
    Weight=@n4.Result
    BoneToModify.0.Transform.Rotation=&RigVMFunction_MathQuaternionFromEuler(Euler=(X=0,Y=0,Z=-30)).Result
   ) {is=struct RigVMUnitNode}
   > &RigUnit_ModifyBoneTransforms(
    BoneToModify=((Bone="dyn_mushroomBeltFrontMain_01",Transform=(Rotation=(X=0,Y=0,Z=0,W=1),Translation=(X=-2,Y=2,Z=0),Scale3D=(X=1,Y=1,Z=1))))
    Mode=AdditiveLocal
    WeightMaximum=1
    Weight=@n4.Result
    BoneToModify.0.Transform.Rotation=&RigVMFunction_MathQuaternionFromEuler(Euler=(X=-30,Y=35,Z=-35)).Result
   )
   n5 = &RigVMFunction_MathFloatAbs(Value=&RigVMFunction_MathQuaternionToEuler(RotationOrder=ZYX, Value=&RigUnit_GetRelativeBoneTransform(Bone=thigh_l, Space=pelvis).Transform.Rotation).Result.Z) {template="Absolute::Execute(in Value,out Result)"}
   n6 = RigVMVariableNode() {variable=is crouching}
   > &RigUnit_ModifyBoneTransforms(
    BoneToModify=((Bone="dyn_skirtFrontTop_01_Base_l",Transform=(Rotation=(X=0,Y=0,Z=0,W=1),Translation=(X=0,Y=-6,Z=0),Scale3D=(X=1,Y=1,Z=1))))
    Mode=AdditiveLocal
    WeightMaximum=1
    Weight=&RigVMFunction_MathFloatMul(
     A=&RigVMFunction_MathFloatRemap(SourceMaximum=102, SourceMinimum=30, TargetMinimum=0, bClamp=true, Value=@n5.Result).Result
     B=@n6.Value
    ) {template="Multiply::Execute(in A,in B,out Result)"}.Result
    BoneToModify.0.Transform.Rotation=&RigVMFunction_MathQuaternionFromEuler(Euler=(X=-45,Y=0,Z=30)).Result
   )
   n7 = &RigVMFunction_MathFloatAbs(Value=&RigVMFunction_MathQuaternionToEuler(RotationOrder=ZYX, Value=&RigUnit_GetRelativeBoneTransform(Bone=thigh_l, Space=pelvis).Transform.Rotation).Result.Z) {template="Absolute::Execute(in Value,out Result)"}
   > &RigUnit_ModifyBoneTransforms(
    BoneToModify=((Bone="dyn_skirtSideBackTop_01_l_Base",Transform=(Rotation=(X=0,Y=0,Z=0,W=1),Translation=(X=0,Y=-2,Z=0),Scale3D=(X=1,Y=1,Z=1))))
    Mode=AdditiveLocal
    WeightMaximum=1
    Weight=&RigVMFunction_MathFloatMul(
     A=&RigVMFunction_MathFloatRemap(SourceMaximum=102, SourceMinimum=30, TargetMinimum=0, bClamp=true, Value=@n7.Result).Result
     B=@n6.Value
    ) {template="Multiply::Execute(in A,in B,out Result)"}.Result
    BoneToModify.0.Transform.Rotation=&RigVMFunction_MathQuaternionFromEuler(Euler=(X=25,Y=-60,Z=0)).Result
   )
   > &RigUnit_TransformConstraint(
    BaseTransform=(Rotation=(X=0,Y=0,Z=0,W=1),Translation=(X=0,Y=0,Z=0),Scale3D=(X=1,Y=1,Z=1))
    BaseTransformSpace=GlobalSpace
    Bone=dyn_armLeaf_B_Base_r
    Targets="((Transform=(Rotation=(X=0,Y=0,Z=0,W=1),Translation=(X=0,Y=0,Z=0),Scale3D=(X=1,Y=1,Z=1)),Weight=0.2,bMaintainOffset=True,Filter=(TranslationFilter=(bX=True,bY=True,bZ=True),RotationFilter=(bX=True,bY=True,bZ=True),ScaleFilter=(bX=True,bY=True,bZ=True))),(T ...(truncated)"
    bUseInitialTransforms=True
    Targets.0.Transform=&RigUnit_GetBoneTransform(Bone=clavicle_r).Transform
    Targets.1.Transform=&RigUnit_GetBoneTransform(Bone=upperarm_twist_01_r).Transform
   )
   > &RigUnit_ModifyBoneTransforms(
    BoneToModify=((Bone="dyn_armLeaf_B_Base_r",Transform=(Rotation=(X=0,Y=0,Z=0,W=1),Translation=(X=0,Y=-0.3,Z=0),Scale3D=(X=1,Y=1,Z=1))))
    Mode=AdditiveLocal
    Weight=1
    WeightMaximum=1
   )
  ]}
]
hierarchy [
 root {loc=<0 0 0>, rot=<0 0 0 1>}
 pelvis {parent=root, loc=<0 -0.5301 98.0571>, rot=<0.0127 0.707 -0.0127 -0.707>}
 spine_01 {parent=pelvis, loc=<6.1208 -0.9328 0>, rot=<0 0 0.034 -0.9994>}
 spine_02 {parent=spine_01, loc=<6.5996 -0.2095 0>, rot=<0 0 0.0238 -0.9997>}
 spine_03 {parent=spine_02, loc=<6.4914 -0.1237 0>, rot=<0 0 0.0149 -0.9999>}
 spine_04 {parent=spine_03, loc=<6.9598 0.0794 0>, rot=<0 0 -0.0132 -0.9999>}
 spine_05 {parent=spine_04, loc=<6.9987 0 0>, rot=<0 0 -0.0015 -1>}
 clavicle_l {parent=spine_05, loc=<5.4831 -0.1053 -3.2946>, rot=<0.0465 -0.7289 -0.009 -0.6829>}
 upperarm_l {parent=clavicle_l, loc=<9.8658 -1.2136 -0.4105>, rot=<0.0164 -0.3245 -0.0269 -0.9454>}
 lowerarm_l {parent=upperarm_l, loc=<26.465 -0.8345 0.259>, rot=<0.0036 0.0161 -0.2298 -0.9731>}
 hand_l {parent=lowerarm_l, loc=<22.7355 -0.3203 -0.6105>, rot=<-0.6784 -0.0223 0.0735 -0.7307>}
 index_metacarpal_l {parent=hand_l, loc=<1.6535 0.0183 -2.96>, rot=<0.0158 -0.0124 -0.0173 -0.9996>}
 index_01_l {parent=index_metacarpal_l, loc=<8.3008 0.0357 -0.0524>, rot=<-0.0044 0.0485 0.0924 -0.9945>}
 index_02_l {parent=index_01_l, loc=<3.3587 -0.3004 0.0452>, rot=<-0.002 0.0027 0.3225 -0.9466>}
 index_03_l {parent=index_02_l, loc=<2.5578 -0.123 -0.0252>, rot=<-0.002 0.0067 0.2558 -0.9667>}
 middle_metacarpal_l {parent=hand_l, loc=<2.1856 -0.1725 -0.5984>, rot=<0.006 0.0168 -0.0329 -0.9993>}
 middle_01_l {parent=middle_metacarpal_l, loc=<7.3275 -0.2447 0.0129>, rot=<-0.0208 0.0042 0.0858 -0.9961>}
 middle_02_l {parent=middle_01_l, loc=<4.0123 0.1401 0.1175>, rot=<-0.0218 0.0258 0.3512 -0.9357>}
 middle_03_l {parent=middle_02_l, loc=<2.6077 0.0534 0.1373>, rot=<0.0082 0.0188 0.1708 -0.9851>}
 pinky_metacarpal_l {parent=hand_l, loc=<1.7616 -0.9753 3.3814>, rot=<-0.0397 0.0814 -0.069 -0.9935>}
 pinky_01_l {parent=pinky_metacarpal_l, loc=<7.1709 -0.4831 -0.1094>, rot=<-0.0654 -0.0035 0.2242 -0.9723>}
 pinky_02_l {parent=pinky_01_l, loc=<2.8274 0.0893 -0.1043>, rot=<0.0037 0.0011 0.2854 -0.9584>}
 pinky_03_l {parent=pinky_02_l, loc=<2.2046 0.0307 -0.0078>, rot=<-0.0002 0.0098 0.1445 -0.9895>}
 ring_metacarpal_l {parent=hand_l, loc=<2.2074 -0.3625 1.3638>, rot=<-0.0284 0.0722 -0.0458 -0.9959>}
 ring_01_l {parent=ring_metacarpal_l, loc=<7.0295 -0.6197 -0.1446>, rot=<-0.0489 -0.0252 0.1892 -0.9804>}
 ring_02_l {parent=ring_01_l, loc=<3.6174 0.4647 -0.1027>, rot=<0 0 0.2806 -0.9598>}
 ring_03_l {parent=ring_02_l, loc=<2.4152 0.114 0.0199>, rot=<0 0 0.1189 -0.9929>}
 thumb_01_l {parent=hand_l, loc=<1.3118 -0.3191 -2.459>, rot=<0.6546 -0.4042 0.0664 -0.6354>}
 thumb_02_l {parent=thumb_01_l, loc=<5.9336 -0.0448 0.0401>, rot=<-0.0037 0.0213 0.298 -0.9543>}
 thumb_03_l {parent=thumb_02_l, loc=<2.2223 0.3682 -0.0907>, rot=<0.0118 -0.013 0.0066 -0.9998>}
 lowerarm_twist_01_l {parent=lowerarm_l, loc=<18.316 0 0>, rot=<0 0 0 1>}
 lowerarm_twist_02_l {parent=lowerarm_l, loc=<10.6707 0 0>, rot=<0 0 0 1>}
 upperarm_twist_01_l {parent=upperarm_l, loc=<7.2135 -0.1134 0>, rot=<0 0 0 1>}
 upperarm_twist_02_l {parent=upperarm_l, loc=<19.7998 -0.1134 0>, rot=<0 0 0 1>}
 dyn_shoulderArmorMain_Base {parent=upperarm_l, loc=<-2.7362 0.2656 -0.7622>, rot=<0.7115 -0.0715 -0.0974 -0.6922>}
 dyn_shoulderArmorMain {parent=dyn_shoulderArmorMain_Base, loc=<0 0 0>, rot=<0 0 0 -1>}
 dyn_shoulderArmorLeaf_A {parent=dyn_shoulderArmorMain, loc=<3.3145 -7.1938 -11.7231>, rot=<-0.3701 0.0419 0.4038 -0.8356>}
 clavicle_r {parent=spine_05, loc=<5.4831 -0.1053 3.2946>, rot=<0.7289 0.0465 -0.6829 0.009>}
 upperarm_r {parent=clavicle_r, loc=<-9.8658 1.2136 0.4104>, rot=<0.0164 -0.3245 -0.0269 -0.9454>}
 lowerarm_r {parent=upperarm_r, loc=<-26.4651 0.8345 -0.259>, rot=<0.0036 0.0161 -0.2298 -0.9731>}
 hand_r {parent=lowerarm_r, loc=<-22.7356 0.3203 0.6106>, rot=<-0.6784 -0.0223 0.0735 -0.7307>}
 index_metacarpal_r {parent=hand_r, loc=<-1.6534 -0.0181 2.9601>, rot=<0.0158 -0.0124 -0.0173 -0.9996>}
 index_01_r {parent=index_metacarpal_r, loc=<-8.3008 -0.0357 0.0524>, rot=<-0.0044 0.0485 0.0924 -0.9945>}
 index_02_r {parent=index_01_r, loc=<-3.3589 0.3003 -0.0453>, rot=<-0.002 0.0027 0.3225 -0.9466>}
 index_03_r {parent=index_02_r, loc=<-2.5578 0.123 0.0252>, rot=<-0.002 0.0067 0.2558 -0.9667>}
 middle_metacarpal_r {parent=hand_r, loc=<-2.1856 0.1726 0.5984>, rot=<0.006 0.0168 -0.0329 -0.9993>}
 middle_01_r {parent=middle_metacarpal_r, loc=<-7.3273 0.2447 -0.0129>, rot=<-0.0208 0.0042 0.0858 -0.9961>}
 middle_02_r {parent=middle_01_r, loc=<-4.0124 -0.14 -0.1175>, rot=<-0.0218 0.0258 0.3512 -0.9357>}
 middle_03_r {parent=middle_02_r, loc=<-2.6077 -0.0533 -0.1373>, rot=<0.0082 0.0188 0.1708 -0.9851>}
 pinky_metacarpal_r {parent=hand_r, loc=<-1.7617 0.9753 -3.3814>, rot=<-0.0397 0.0814 -0.069 -0.9935>}
 pinky_01_r {parent=pinky_metacarpal_r, loc=<-7.1707 0.4832 0.1094>, rot=<-0.0654 -0.0035 0.2242 -0.9723>}
 pinky_02_r {parent=pinky_01_r, loc=<-2.8274 -0.0892 0.1043>, rot=<0.0037 0.0011 0.2854 -0.9584>}
 pinky_03_r {parent=pinky_02_r, loc=<-2.2045 -0.031 0.0078>, rot=<-0.0002 0.0098 0.1445 -0.9895>}
 ring_metacarpal_r {parent=hand_r, loc=<-2.2073 0.3627 -1.3638>, rot=<-0.0284 0.0722 -0.0458 -0.9959>}
 ring_01_r {parent=ring_metacarpal_r, loc=<-7.0297 0.6197 0.1446>, rot=<-0.0489 -0.0252 0.1892 -0.9804>}
 ring_02_r {parent=ring_01_r, loc=<-3.6171 -0.4648 0.1027>, rot=<0 0 0.2806 -0.9598>}
 ring_03_r {parent=ring_02_r, loc=<-2.4153 -0.1141 -0.0199>, rot=<0 0 0.1189 -0.9929>}
 thumb_01_r {parent=hand_r, loc=<-1.3119 0.3192 2.459>, rot=<0.6546 -0.4042 0.0664 -0.6354>}
 thumb_02_r {parent=thumb_01_r, loc=<-5.9335 0.0447 -0.0401>, rot=<-0.0037 0.0213 0.298 -0.9543>}
 thumb_03_r {parent=thumb_02_r, loc=<-2.2223 -0.3682 0.0907>, rot=<0.0118 -0.013 0.0066 -0.9998>}
 lowerarm_twist_01_r {parent=lowerarm_r, loc=<-18.3161 0 0>, rot=<0 0 0 1>}
 lowerarm_twist_02_r {parent=lowerarm_r, loc=<-10.6707 0 0>, rot=<0 0 0 1>}
 upperarm_twist_01_r {parent=upperarm_r, loc=<-7.2135 0.1134 0>, rot=<0 0 0 1>}
 dyn_armLeaf_C_r {parent=upperarm_twist_01_r, loc=<8.2096 3.1299 -3.9899>, rot=<-0.0669 0.3601 0.9304 -0.0109>}
 dyn_armLeaf_B_Base_r {parent=upperarm_twist_01_r, loc=<5.16 -2.1539 -4.9039>, rot=<0.0106 -0.9342 -0.3243 -0.1482>}
 dyn_armLeaf_B_r {parent=dyn_armLeaf_B_Base_r, loc=<0 0 0>, rot=<0 0 0 1>}
 dyn_armLeaf_B_Back_r {parent=dyn_armLeaf_B_Base_r, loc=<0 0 0>, rot=<0.1701 -0.7816 0.565 -0.2022>}
 upperarm_twist_02_r {parent=upperarm_r, loc=<-19.7998 0.1134 0>, rot=<0 0 0 1>}
 dyn_armLeaf_D_r {parent=clavicle_r, loc=<-8.5601 8.3674 3.2532>, rot=<0.0678 0.2431 0.9601 0.1208>}
 neck_01 {parent=spine_05, loc=<11.3138 0.1529 0>, rot=<0 0 -0.2003 -0.9797>}
 neck_02 {parent=neck_01, loc=<4.5627 0 0>, rot=<0 0 0.0504 -0.9987>}
 head {parent=neck_02, loc=<5.3914 0 0>, rot=<0 0 0.115 -0.9934>}
 dyn_necklaceMain_Base {parent=spine_05, loc=<8.0829 5.2267 0.7317>, rot=<0.0032 -0.0073 0.9443 0.329>}
 dyn_necklaceMain {parent=dyn_necklaceMain_Base, loc=<0 0 0>, rot=<0 0 0 1>}
 dyn_torsoLeaf_A_01_r {parent=spine_02, loc=<6.5265 -5.8854 7.654>, rot=<0.9071 -0.3439 0.0741 0.2312>}
 dyn_torsoLeaf_A_02_r {parent=dyn_torsoLeaf_A_01_r, loc=<3.6345 0 0>, rot=<0.2326 -0.0491 -0.2878 -0.9277>}
 dyn_frontGrass_Base {parent=spine_02, loc=<1.8662 6.8188 -7.6034>, rot=<0.2368 0.073 -0.1809 -0.9518>}
 dyn_frontGrass {parent=dyn_frontGrass_Base, loc=<0 0 0>, rot=<0 0 0 -1>}
 dyn_frontFlowers {parent=spine_02, loc=<-2.8708 6.2532 -9.3841>, rot=<-0.2769 0.3655 0.8563 0.2377>}
 thigh_l {parent=pelvis, loc=<-1.0855 -0.2109 -8.7053>, rot=<0.0009 0.0355 0.0155 -0.9993>}
 calf_l {parent=thigh_l, loc=<-42.5284 -0.0019 -0.7297>, rot=<0.0006 -0.0161 -0.0391 -0.9991>}
 calf_twist_01_l {parent=calf_l, loc=<-35.8483 -1.9197 -0.5794>, rot=<0.0237 0.0095 -0.0035 -0.9997>}
 calf_twist_02_l {parent=calf_l, loc=<-19.1546 -1.7943 -0.2638>, rot=<0.0237 0.0095 -0.0035 -0.9997>}
 foot_l {parent=calf_l, loc=<-46.7401 -2.5014 -0.592>, rot=<0.0427 -0.0238 0.0353 -0.9982>}
 ball_l {parent=foot_l, loc=<-3.5658 10.8066 -0.1544>, rot=<0.0001 -0.0001 -0.7063 -0.7079>}
 dyn_kneeAim_l {parent=calf_l, loc=<0 0 0>, rot=<0 0 0 1>}
 thigh_twist_01_l {parent=thigh_l, loc=<-27.952 0 0>, rot=<0.0013 0 0 -1>}
 dyn_skirtSideFore_01_l {parent=thigh_l, loc=<-20.4285 4.9824 -6.4435>, rot=<0.3139 0.0114 -0.0197 -0.9492>}
 dyn_skirtSideFore_02_l {parent=dyn_skirtSideFore_01_l, loc=<-12.188 0 -0.0001>, rot=<0.0527 -0.0055 0.057 -0.997>}
 dyn_skirtSideFore_03_l {parent=dyn_skirtSideFore_02_l, loc=<-10.9637 0 0>, rot=<0 0 0 -1>}
 dyn_skirtSideMid_01_l {parent=thigh_l, loc=<-29.1451 -0.7674 -10.8978>, rot=<0.7282 0.0499 0.0468 -0.6819>}
 dyn_skirtSideMid_02_l {parent=dyn_skirtSideMid_01_l, loc=<-11.1685 0 0>, rot=<0 0 0 -1>}
 dyn_skirtSideBack_01_l {parent=thigh_l, loc=<-30.822 -8.3706 -6.9454>, rot=<0.8317 0.114 0.0592 -0.5401>}
 dyn_skirtSideBack_02_l {parent=dyn_skirtSideBack_01_l, loc=<-12.1193 0 -0.0001>, rot=<0 0 0 1>}
 dyn_legArmor_01_l {parent=thigh_l, loc=<4.84 -1.2903 -5.0167>, rot=<-0.0935 0.5954 0.7956 0.0614>}
 dyn_legArmor_02_l {parent=dyn_legArmor_01_l, loc=<19.5115 0 0>, rot=<-0.0255 -0.0081 0.0933 -0.9953>}
 dyn_legArmorLeaf_A_l {parent=dyn_legArmor_02_l, loc=<15.4054 -1.531 -7.2416>, rot=<0.1529 -0.4214 0.8612 -0.2396>}
 dyn_legArmorSide_01_l {parent=dyn_legArmor_01_l, loc=<2.2616 0.3416 6.0008>, rot=<0.1176 0.0205 0.029 -0.9924>}
 dyn_hipBag_Base_l {parent=dyn_legArmorSide_01_l, loc=<1.1024 -0.5264 0.7101>, rot=<0.0441 0.0595 0.0499 -0.996>}
 dyn_hipBag_l {parent=dyn_hipBag_Base_l, loc=<0 0 0>, rot=<0 0 0 -1>}
 thigh_r {parent=pelvis, loc=<-1.0853 -0.2109 8.7053>, rot=<0.0355 -0.0009 0.9993 0.0155>}
 calf_r {parent=thigh_r, loc=<42.5286 0.0019 0.7297>, rot=<0.0006 -0.0161 -0.0391 -0.9991>}
 calf_twist_01_r {parent=calf_r, loc=<35.8485 1.9197 0.5793>, rot=<0.0237 0.0095 -0.0035 -0.9997>}
 calf_twist_02_r {parent=calf_r, loc=<19.1546 1.7943 0.2638>, rot=<0.0237 0.0095 -0.0035 -0.9997>}
 foot_r {parent=calf_r, loc=<46.7403 2.5014 0.5919>, rot=<0.0427 -0.0238 0.0353 -0.9982>}
 ball_r {parent=foot_r, loc=<3.5658 -10.8066 0.1544>, rot=<0.0001 -0.0001 -0.7063 -0.7079>}
 thigh_twist_01_r {parent=thigh_r, loc=<27.952 0 0>, rot=<0.0013 0 0 -1>}
 dyn_skirtSideFore_01_r {parent=thigh_r, loc=<20.4286 -4.9824 6.4435>, rot=<0.3139 0.0114 -0.0197 -0.9492>}
 dyn_skirtSideFore_02_r {parent=dyn_skirtSideFore_01_r, loc=<12.188 0 0>, rot=<0.0527 -0.0055 0.057 -0.997>}
 dyn_skirtSideFore_03_r {parent=dyn_skirtSideFore_02_r, loc=<10.9637 0 0>, rot=<0 0 0 -1>}
 dyn_skirtSideMid_01_r {parent=thigh_r, loc=<29.1452 0.7674 10.8978>, rot=<0.7282 0.0499 0.0468 -0.6819>}
 dyn_skirtSideMid_02_r {parent=dyn_skirtSideMid_01_r, loc=<11.1685 0 0>, rot=<0 0 0 1>}
 dyn_skirtSideBack_01_r {parent=thigh_r, loc=<30.8222 8.3706 6.9454>, rot=<0.8317 0.114 0.0592 -0.5401>}
 dyn_skirtSideBack_02_r {parent=dyn_skirtSideBack_01_r, loc=<12.1193 0 0>, rot=<0 0 0 -1>}
 dyn_legArmor_01_r {parent=thigh_r, loc=<-3.1055 3.2464 6.3469>, rot=<0.7 0.0702 0.0748 -0.7067>}
 dyn_legArmor_02_r {parent=dyn_legArmor_01_r, loc=<19.3943 0 0>, rot=<-0.0116 -0.0046 0.1025 -0.9947>}
 dyn_skirtBackTopMain_c_Base {parent=pelvis, loc=<-7.1656 -13.2217 0.3244>, rot=<0.049 -0.7054 0.7054 -0.049>}
 dyn_skirtBackTopMain_c {parent=dyn_skirtBackTopMain_c_Base, loc=<0 0 0>, rot=<0 0 0 1>}
 dyn_skirtBackTop_01_c {parent=dyn_skirtBackTopMain_c, loc=<14.4918 0.1515 -1.1456>, rot=<0.7492 -0.0272 0.0268 -0.6612>}
 dyn_skirtBackTop_02_c {parent=dyn_skirtBackTop_01_c, loc=<17.5408 1.061 1.2572>, rot=<-0.0249 0.0658 -0.0413 -0.9967>}
 dyn_skirtSideBackTop_01_l_Base {parent=pelvis, loc=<-6.1593 -7.985 -13.6538>, rot=<0.8917 0.0841 0.0576 -0.441>}
 dyn_skirtSideBackTop_01_l {parent=dyn_skirtSideBackTop_01_l_Base, loc=<0 0 0>, rot=<0 0 0 -1>}
 dyn_skirtSideBackTop_02_l {parent=dyn_skirtSideBackTop_01_l, loc=<-16.6777 0.4464 -0.3909>, rot=<0 0 0 -1>}
 dyn_skirtSideBackTop_01_r_Base {parent=pelvis, loc=<-6.1593 -7.985 13.6538>, rot=<0.0841 -0.8917 0.441 0.0576>}
 dyn_skirtSideBackTop_01_ra {parent=dyn_skirtSideBackTop_01_r_Base, loc=<0 0 0>, rot=<0 0 0 1>}
 dyn_skirtSideBackTop_02_ra {parent=dyn_skirtSideBackTop_01_ra, loc=<16.6776 -0.4464 0.3909>, rot=<0 0 0 1>}
 dyn_skirtFrontMain_BaseUpper {parent=pelvis, loc=<-0.0581 9.0519 0.0604>, rot=<0 0 0.9998 0.0179>}
 dyn_skirtFrontMain_Base {parent=dyn_skirtFrontMain_BaseUpper, loc=<15.5871 0 0>, rot=<0 0 0 -1>}
 dyn_skirtFrontMain {parent=dyn_skirtFrontMain_Base, loc=<0 0 0>, rot=<0 0 0 -1>}
 dyn_skirtFrontBottom_01_l {parent=dyn_skirtFrontMain, loc=<11.8017 -1.1978 0.1509>, rot=<-0.0309 0.014 0.0424 -0.9985>}
 dyn_skirtFrontBottom_02_l {parent=dyn_skirtFrontBottom_01_l, loc=<15.583 0.2917 -0.5954>, rot=<-0.0066 0.0007 -0.0183 -0.9998>}
 dyn_legArmorFrontMain_Base {parent=pelvis, loc=<0.7651 9.9571 2.1184>, rot=<-0.0506 -0.004 0.9921 0.1143>}
 dyn_legArmorFrontMain {parent=dyn_legArmorFrontMain_Base, loc=<0 0 0>, rot=<0 0 0 1>}
 dyn_legArmorFront_02_l {parent=dyn_legArmorFrontMain, loc=<7.9957 -0.6165 -0.7512>, rot=<0.0033 -0.0195 -0.0247 -0.9995>}
 dyn_backArmor_01 {parent=pelvis, loc=<1.5778 -12.4811 0.6961>, rot=<0.0978 -0.9951 -0.0119 -0.0016>}
 dyn_backArmor_02 {parent=dyn_backArmor_01, loc=<12.3738 0 0>, rot=<0.0925 0.0237 0.0196 -0.9952>}
 dyn_backArmorLeaf_A {parent=dyn_backArmor_01, loc=<7.3596 1.0403 13.0416>, rot=<0.0987 0.1633 0.1843 -0.9642>}
 dyn_backArmorLeaf_B {parent=dyn_backArmor_01, loc=<10.3081 2.8761 12.2672>, rot=<-0.1422 0.4701 0.2324 -0.8395>}
 dyn_rope_01_Base {parent=pelvis, loc=<5.161 12.2556 -3.0689>, rot=<-0.0083 0.0554 0.9816 0.1824>}
 dyn_rope_01 {parent=dyn_rope_01_Base, loc=<0 0 0>, rot=<0 0 0 -1>}
 dyn_rope_02 {parent=dyn_rope_01, loc=<6.6222 0 0>, rot=<0.016 -0.019 -0.0969 -0.995>}
 dyn_rope_03 {parent=dyn_rope_02, loc=<6.0808 0 0>, rot=<-0.008 -0.0083 -0.0336 -0.9994>}
 dyn_rope_04 {parent=dyn_rope_03, loc=<6.4226 0 0>, rot=<0.0279 0.0635 0.0993 -0.9926>}
 dyn_mushroomBeltSideMain_01_Base {parent=pelvis, loc=<7.4213 -3.4021 13.9529>, rot=<0.1313 -0.6796 0.7033 0.1622>}
 dyn_mushroomBeltSideMain_01 {parent=dyn_mushroomBeltSideMain_01_Base, loc=<0 0 0>, rot=<0 0 0 -1>}
 dyn_mushroomBeltSideMain_02 {parent=dyn_mushroomBeltSideMain_01, loc=<6.0415 0 0>, rot=<0 0 0 -1>}
 dyn_mushroom_E {parent=dyn_mushroomBeltSideMain_02, loc=<-2.518 0.5971 7.0638>, rot=<0.2629 0.1136 0.0508 -0.9568>}
 dyn_mushroom_D {parent=dyn_mushroomBeltSideMain_02, loc=<0.1518 -1.3178 -7.4483>, rot=<-0.0787 -0.0216 0.101 -0.9915>}
 dyn_mushroomBeltFrontMain_01_Base {parent=pelvis, loc=<9.67 6.4422 8.9519>, rot=<0.0778 -0.317 0.9093 0.258>}
 dyn_mushroomBeltFrontMain_01 {parent=dyn_mushroomBeltFrontMain_01_Base, loc=<0 0 0>, rot=<0 0 0 1>}
 dyn_mushroomBeltFrontMain_02 {parent=dyn_mushroomBeltFrontMain_01, loc=<9.1856 0 0>, rot=<0 0 0 1>}
 dyn_mushroom_A {parent=dyn_mushroomBeltFrontMain_02, loc=<-2.3185 1.5283 -6.9484>, rot=<-0.243 -0.1699 -0.1486 -0.9434>}
 dyn_mushroom_C {parent=dyn_mushroomBeltFrontMain_02, loc=<-0.5702 -0.5636 0.2565>, rot=<0.0771 0.1056 0.0331 -0.9909>}
 dyn_backBottomGrass_Base {parent=pelvis, loc=<4.7316 -12.3065 -5.6738>, rot=<0.1733 -0.9531 -0.1896 -0.1601>}
 dyn_backBottomGrassB {parent=dyn_backBottomGrass_Base, loc=<0 0 0>, rot=<0 0 0 -1>}
 dyn_backBottomGrassC {parent=dyn_backBottomGrass_Base, loc=<0 0 0>, rot=<-0.0833 -0.5741 0.1476 -0.801>}
 dyn_backTopGrassMain_Base {parent=pelvis, loc=<14.0611 -9.7656 -4.6482>, rot=<0.915 -0.3113 -0.2503 0.0565>}
 dyn_backTopGrassA {parent=dyn_backTopGrassMain_Base, loc=<0.6122 1.0461 0.2602>, rot=<-0.251 0.5371 0.0757 -0.8017>}
 dyn_backTopGrassB {parent=dyn_backTopGrassMain_Base, loc=<0 0 0>, rot=<-0.2588 0 0 -0.9659>}
 dyn_backTopGrassH {parent=dyn_backTopGrassMain_Base, loc=<0.3776 1.3887 -0.9302>, rot=<0.0365 -0.1822 -0.6736 -0.7154>}
 dyn_backBottomFlowerB {parent=dyn_backTopGrassMain_Base, loc=<-1.3016 4.5182 2.9692>, rot=<-0.2576 0.1593 -0.7221 -0.622>}
]
variables [is crouching {type=double, default=0}]
