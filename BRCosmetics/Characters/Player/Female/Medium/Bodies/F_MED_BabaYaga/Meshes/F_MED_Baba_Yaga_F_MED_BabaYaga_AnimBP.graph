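# Post-processing-style AnimBlueprint for the Baba Yaga medium female body cosmetic, adding secondary motion to armor, leaves and mushroom.
# Copies the attached parent's pose through a crouch-aware Control Rig, three chained RigidBody sims, leaf AnimDynamics and bone tweaks.
# Crouch movement and jump/fall states drive mushroom adjustments; a deprecated helper sets gravity for the first sim.
# Dynamics reset after customization loads and crouch ends; the cape flag is unused here.
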AnimBlueprint /BRCosmetics/Characters/Player/Female/Medium/Bodies/F_MED_BabaYaga/Meshes/F_MED_Baba_Yaga_F_MED_BabaYaga_AnimBP
target_skeleton=&/F_MED_BabaYaga_Skeleton
graphs [
 AnimGraph {nodes [
   Anim Dynamics - Bone: dyn_backArmorLeaf_A(Alpha=1, ExternalForce="0,0,0") {
    angularDampingOverride=200
    angularSpringConstant=300
    bAngularSpring=true
    bGravityOverrideInSimSpace=true
    bUseGravityOverride=true
    simulationSpace=BoneRelative
    alphaBoolBlend {blendOption=Linear}
    boundBone {boneName=dyn_backArmorLeaf_A}
    componentAppliedLinearAccClamp=<1500 1500 1500>
    componentLinearAccScale=<0.4 0.4 1>
    componentLinearVelScale=<0.2 0.2 0.4>
    gravityOverride {z=-980}
    physicsBodyDefinitions [{
      sphereCollisionRadius=0
      boundBone {boneName=dyn_backArmorLeaf_A}
      boxExtents=<2 5 5>
      constraintSetup {angularLimitsMax=<8 5 8>, angularLimitsMin=<-8 -5 -8>, angularTarget=<1 0 0>}
      localJointOffset=<0.125 0 0>
     }]
    relativeSpaceBone {boneName=root}
   }
   Anim Dynamics - Bone: dyn_legArmorLeaf_A_l(Alpha=1, ExternalForce="0,0,0") {
    angularDampingOverride=200
    angularSpringConstant=100
    bAngularSpring=true
    bGravityOverrideInSimSpace=true
    bUseGravityOverride=true
    simulationSpace=BoneRelative
    alphaBoolBlend {blendOption=Linear}
    boundBone {boneName=dyn_legArmorLeaf_A_l}
    componentAppliedLinearAccClamp=<1500 1500 1500>
    componentLinearAccScale=<0.2 0.2 0.5>
    componentLinearVelScale=<0.2 0.2 0.4>
    gravityOverride {z=-980}
    physicsBodyDefinitions [{
      sphereCollisionRadius=0
      boundBone {boneName=dyn_legArmorLeaf_A_l}
      boxExtents=<2 7 7>
      constraintSetup {angularLimitsMax=<8 15 0>, angularLimitsMin=<-8 -15 -30>, angularTarget=<1 0 0>}
      localJointOffset=<1 0 0>
     }]
    relativeSpaceBone {boneName=root}
   }
   Anim Dynamics - Bone: dyn_backArmorLeaf_A(Alpha=1, ExternalForce="0,0,0") {
    angularDampingOverride=200
    angularSpringConstant=500
    bAngularSpring=true
    bGravityOverrideInSimSpace=true
    bUseGravityOverride=true
    simulationSpace=BoneRelative
    alphaBoolBlend {blendOption=Linear}
    boundBone {boneName=dyn_backArmorLeaf_A}
    componentAppliedLinearAccClamp=<1500 1500 1500>
    componentLinearAccScale=<0.4 0.4 1>
    componentLinearVelScale=<0.2 0.2 0.4>
    gravityOverride {z=-980}
    physicsBodyDefinitions [{
      sphereCollisionRadius=0
      boundBone {boneName=dyn_backArmorLeaf_A}
      boxExtents=<7 7 7>
      constraintSetup {angularLimitsMax=<8 5 8>, angularLimitsMin=<-8 -5 -8>, angularTarget=<1 0 0>}
      localJointOffset=<0.2 0 0>
     }]
    relativeSpaceBone {boneName=root}
   }
   n1 = Control Rig!(Source=Copy Pose From Mesh() {bUseAttachedParent=true}, Alpha, is crouching:double=Get crouchAlpha() -> (crouchAlpha:double)) -> (Pose) {
    alphaBoolBlend {blendOption=Linear}
    controlRigAssetReference {blueprintRigClass=&/F_MED_Baba_Yaga_CtrlRig.F_MED_Baba_Yaga_CtrlRig_C}
   }
   n2 = RigidBody!(ComponentPose=Local To Component(LocalPose=@n1), Alpha, OverrideWorldGravity:Vector=Get GravityOverride(), ExternalForce) -> (Pose) {
    bClampLinearTranslationLimitToRefPose=true
    bOverrideWorldGravity=true
    overridePhysicsAsset=&/F_MED_BabaYaga_Physics
    simulationSpace=BaseBoneSpace
    alphaBoolBlend {blendOption=Linear}
    baseBoneRef {boneName=root}
    componentAppliedLinearAccClamp=<2000 2000 2000>
    componentLinearAccScale=<0.3 0.3 0.3>
    componentLinearVelScale=<0.1 0.1 0.1>
    overrideWorldGravity {z=-980}
   }
   n3 = RigidBody!(ComponentPose=@n2, Alpha, ExternalForce) -> (Pose) {
    bClampLinearTranslationLimitToRefPose=true
    bOverrideWorldGravity=true
    overridePhysicsAsset=&/F_MED_BabaYaga_LeafPhysics
    simulationSpace=BaseBoneSpace
    alphaBoolBlend {blendOption=Linear}
    baseBoneRef {boneName=root}
    componentAppliedLinearAccClamp=<2000 2000 2000>
    componentLinearAccScale=<0.3 0.3 0.4>
    componentLinearVelScale=<0.1 0.1 0.1>
    overrideWorldGravity {z=-980}
   }
   n4 = RigidBody(ComponentPose=@n3) {
    bClampLinearTranslationLimitToRefPose=true
    bOverrideWorldGravity=true
    overridePhysicsAsset=&/F_MED_BabaYaga_ExtrasPhysics
    simulationSpace=BaseBoneSpace
    alphaBoolBlend {blendOption=Linear}
    baseBoneRef {boneName=pelvis}
    componentAppliedLinearAccClamp=<2000 2000 2000>
    componentLinearAccScale=<0.2 0.5 0.5>
    componentLinearVelScale=<0.25 0.25 0.35>
    overrideWorldGravity {z=-980}
   }
   n5 = Anim Dynamics - Bone: dyn_legArmorLeaf_A_l(ComponentPose=@n4, Alpha=1, ExternalForce="0,0,0") {
    angularDampingOverride=200
    angularSpringConstant=300
    bAngularSpring=true
    bGravityOverrideInSimSpace=true
    bUseGravityOverride=true
    simulationSpace=BoneRelative
    alphaBoolBlend {blendOption=Linear}
    boundBone {boneName=dyn_legArmorLeaf_A_l}
    componentAppliedLinearAccClamp=<1500 1500 1500>
    componentLinearAccScale=<0.2 0.2 0.5>
    componentLinearVelScale=<0.2 0.2 0.4>
    gravityOverride {z=-980}
    physicsBodyDefinitions [{
      sphereCollisionRadius=0
      boundBone {boneName=dyn_legArmorLeaf_A_l}
      boxExtents=<1 5 5>
      constraintSetup {angularLimitsMax=<16 10 16>, angularLimitsMin=<-16 -10 -16>, angularTarget=<1 0 0>}
      localJointOffset=<0.5 0 0>
     }]
    relativeSpaceBone {boneName=root}
   }
   n6 = Anim Dynamics - Bone: dyn_backArmorLeaf_A(ComponentPose=@n5, Alpha=1, ExternalForce="0,0,0") {
    angularDampingOverride=200
    angularSpringConstant=100
    bAngularSpring=true
    bGravityOverrideInSimSpace=true
    bUseGravityOverride=true
    simulationSpace=BoneRelative
    alphaBoolBlend {blendOption=Linear}
    boundBone {boneName=dyn_backArmorLeaf_A}
    componentAppliedLinearAccClamp=<1500 1500 1500>
    componentLinearAccScale=<0.2 0.2 0.5>
    componentLinearVelScale=<0.2 0.2 0.4>
    gravityOverride {z=-980}
    physicsBodyDefinitions [{
      sphereCollisionRadius=0
      boundBone {boneName=dyn_backArmorLeaf_A}
      boxExtents=<2 7 7>
      constraintSetup {angularLimitsMax=<8 15 0>, angularLimitsMin=<-8 -15 -30>, angularTarget=<1 0 0>}
      localJointOffset=<1 0 0>
     }]
    relativeSpaceBone {boneName=root}
   }
   n7 = "Transform (Modify) Bone - Bone: dyn_rope_01_Base"!(ComponentPose=@n6, Alpha=1, Rotation="-21.082,6.645,-22.529") -> (Pose) {
    rotationMode=BMM_Additive
    rotationSpace=BCS_BoneSpace
    translationMode=BMM_Additive
    translationSpace=BCS_BoneSpace
    alphaBoolBlend {blendOption=Linear}
    alphaScaleBiasClamp {bInterpResult=true}
    boneToModify {boneName=dyn_rope_01_Base}
    rotation=<rot -21.082 6.645 -22.529>
    translation=<-0.1806 -2.0272 -0.7261>
   }
   n8 = "Constraint - dyn_legArmorFrontMain (T)"(ComponentPose=@n7, Alpha=0.3, ConstraintWeights_0=1) {
    alphaBoolBlend {blendOption=Linear}
    boneToModify {boneName=dyn_legArmorFrontMain}
    constraintSetup [{transformType=Translation, targetBone {boneName=dyn_skirtFrontMain}}]
    constraintWeights [1]
   }
   n9 = "Constraint - dyn_mushroom_D (P)"(ComponentPose=@n8, Alpha=Get rotateMushroom() -> (rotateMushroom:double), ConstraintWeights_0=1) {
    alphaBoolBlend {blendOption=Linear}
    alphaScaleBiasClamp {bInterpResult=true}
    boneToModify {boneName=dyn_mushroom_D}
    constraintSetup [{targetBone {boneName=thigh_r}}]
    constraintWeights [1]
   }
   n10 = "Transform (Modify) Bone - Bone: dyn_mushroom_D"(ComponentPose=@n9, Alpha=Get mushroomOnJump() -> (mushroomOnJump:double), Translation="0,0,0", Rotation="0,-80,0", Scale="1,1,1") {
    rotationMode=BMM_Additive
    rotationSpace=BCS_BoneSpace
    translationMode=BMM_Additive
    translationSpace=BCS_BoneSpace
    alphaBoolBlend {blendOption=Linear}
    alphaScaleBiasClamp {bInterpResult=true}
    boneToModify {boneName=dyn_mushroom_D}
    rotation {yaw=-80}
   }
   Output Pose(Result=Component To Local(ComponentPose=@n10)) {layerGroup=None}
  ]}
 EventGraph {nodes [
   Event BlueprintUpdateAnimation()
   > n1 = Sequence!(execute) -> (then_0, then_1, then_2:Exec)
   n2 = Branch(execute=@n1.then_1, Condition=AND Boolean(A=Get bIsAccelerating2D(), B=Get bIsCrouching()))
   > Set rotateMushroom(rotateMushroom:double=0.7) -> (Output_Get:double)
   Set rotateMushroom(execute=@n2.else, rotateMushroom=0)
   n3 = Gravity Override State Function DEPRECATED(
    execute=@n1.then_2
    CharacterPartAnimInstance=Self-Reference()
    GameTimeInSeconds=GetGameTimeInSeconds()
    GravityOverrideParams=Get Gravity Override Params()
   )
   > Set GravityOverride(GravityOverride=@n3)
   n4 = Get bIsJumping()
   n5 = Get bIsFalling()
   Set mushroomOnJump(execute=@n1.then_0, mushroomOnJump:double="To Float (Boolean)"(InBool=OR Boolean(A=@n4, B=@n5))) -> (Output_Get:double)
   Event OnBeginCrouching()
   > Set crouchAlpha(crouchAlpha:double="To Float (Boolean)"(InBool=Get bIsCrouching())) -> (Output_Get:double)
   Event OnCharacterCustomizationLoaded()
   > n6 = Sequence!(execute) -> (then_0, then_1)
   > Do Once()
   > Set isWearingCape(isWearingCape=CheckForBackbling().bIsBackblingACape)
   Delay(execute=@n6.then_1, Duration=0.1)
   > ResetDynamics(InTeleportType=TeleportPhysics)
   > ResetDynamics(InTeleportType=ResetPhysics)
   Event OnEndCrouching()
   > Set crouchAlpha(crouchAlpha="To Float (Boolean)"(InBool=Get bIsCrouching()))
   > Delay(Duration=0.01)
   > ResetDynamics(InTeleportType=TeleportPhysics)
   > ResetDynamics(InTeleportType=ResetPhysics)
  ]}
]
variables [
 capeVectorLinearAcc {type=Vector}
 CapeVectorLinearVel {type=Vector}
 crouchAlpha {type=number}
 Gravity Override Params {type=GravityOverrideParamsStruct}
 GravityOverride {type=Vector}
 isWearingCape {type=boolean}
 mushroomOnJump {type=number}
 rotateMushroom {type=number}
]
