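# Physics asset for the F_MED_BabaYaga female character body, simulating dynamic hair, leaves and hair ties against head and upper-spine colliders.
# Box and capsule bodies form chained front and back locks, with back side chains branching from the central one.
# Most hair joints use tight swing and twist limits, some locked, while spine and head joints are free.
# The right front base attaches freely to spine_03, unlike the left, which attaches to head.
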PhysicsAsset /BRCosmetics/Characters/Player/Female/Medium/Bodies/F_MED_BabaYaga/Meshes/Parts/F_MED_BabaYaga_Physics
bodies {
 dyn_hairBackA_01 [{shapeType=Box, extentX=3.9047, extentY=4.8842, extentZ=9.754, center=<2 0 0>, rotation=<rot 0 90 90>}]
 dyn_hairBackA_02 [{shapeType=Box, extentX=2.2352, extentY=2.6138, extentZ=6.2657, center=<2 0 0>, rotation=<rot 0 90 90>}]
 dyn_hairBack_01_c [{
   shapeType=Box
   extentX=1.1722
   extentY=29.0908
   extentZ=14.9447
   shapeName=dyn_hairBack_01_c_box
   center=<8.545 -0.5607 1.6694>
   rotation=<rot 3.51 -111.0555 9.0371>
  }]
 dyn_hairBack_01_l [{shapeType=Capsule, length=5.3455, radius=2.9063, shapeName=dyn_hairBack_01_l_capsule, center=<4.3205 0.4774 0.4497>, rotation=<rot 90 0 0>}]
 dyn_hairBack_01_r [{shapeType=Capsule, length=7.6235, radius=4.7917, shapeName=dyn_hairBack_01_r_capsule, center=<2.4532 -3.2731 0.9845>, rotation=<rot 90 0 0>}]
 dyn_hairBack_02_c [{shapeType=Box, extentX=3.42, extentY=12.6683, extentZ=9.6093, center=<4.544 0 -0.3286>, rotation=<rot 0 89.9999 82.6406>}]
 dyn_hairBack_02_l [{shapeType=Capsule, length=5.9358, radius=3.2, shapeName=dyn_hairBack_02_l_capsule, center=<2.2451 0 0>, rotation=<rot 90 0 0>}]
 dyn_hairBack_02_r [{
   shapeType=Capsule
   length=5.311
   radius=4.3196
   shapeName=dyn_hairBack_02_r_capsule
   center=<2.9294 -2.3301 -0.3389>
   rotation=<rot 90 0 -8.4927>
  }]
 dyn_hairBack_03_c [{shapeType=Box, extentX=2.5261, extentY=8.0859, extentZ=11.6675, center=<4.692 -0.6382 0>, rotation=<rot 0 87.3041 89.9999>}]
 dyn_hairBack_03_l [{shapeType=Capsule, length=2.6369, radius=2.0836, shapeName=dyn_hairBack_03_l_capsule, center=<1.8626 -0.0606 -0.4096>, rotation=<rot 90 0 0>}]
 dyn_hairBack_03_r [{
   shapeType=Capsule
   length=3.8593
   radius=4.2972
   shapeName=dyn_hairBack_03_r_capsule
   center=<2.9634 -0.1309 0.1684>
   rotation=<rot 0 -94.7689 -82.3278>
  }]
 dyn_hairBack_04_c [{
   shapeType=Capsule
   length=5.8115
   radius=2.3612
   shapeName=dyn_hairBack_04_c_capsule
   center=<3.4266 -0.1284 -0.1476>
   rotation=<rot 0 -85.9121 -87.963>
  }]
 dyn_hairBack_04_l [{
   shapeType=Capsule
   length=3.4219
   radius=1.6199
   shapeName=dyn_hairBack_04_l_capsule
   center=<2.3503 0.0197 -0.1298>
   rotation=<rot 0 90.5719 -79.5793>
  }]
 dyn_hairBack_05_c [{
   shapeType=Capsule
   length=5.9631
   radius=2.0369
   shapeName=dyn_hairBack_05_c_capsule
   center=<3.9504 -0.0257 -0.0323>
   rotation=<rot 0 92.8156 -87.4426>
  }]
 dyn_hairFront_01_Base_l [{shapeType=Capsule, length=0.505, radius=0.505, rotation=<rot 90 0 0>}]
 dyn_hairFront_01_Base_r [{shapeType=Capsule, length=0.505, radius=0.505, center=<4.1198 0.0002 0>, rotation=<rot 90 0 0>}]
 dyn_hairFront_01_l [{shapeType=Capsule, length=4.9842, radius=3.5093, center=<4.1193 -0.0049 -0.0097>, rotation=<rot 65 90 90>}]
 dyn_hairFront_01_r [{shapeType=Capsule, length=5.6881, radius=4.2027, center=<5.1488 -1.7339 0.0001>, rotation=<rot 90 8.8807 13.1367>}]
 dyn_hairFront_02_l [{shapeType=Capsule, length=7.919, radius=3.1065, center=<5.3018 -1.3702 0.642>, rotation=<rot 61.9992 89.9996 108.8469>}]
 dyn_hairFront_02_r [{shapeType=Capsule, length=4.515, radius=3.2282, center=<4.1198 -1.1895 0.0001>, rotation=<rot 90 0 0>}]
 dyn_hairFront_03_l [{shapeType=Capsule, length=6.2203, radius=2.7674, center=<4.1815 -1 0>, rotation=<rot 64.9999 89.9999 89.9999>}]
 dyn_hairFront_03_r [{shapeType=Capsule, length=4.6784, radius=2.3642, center=<4.1198 0 0>, rotation=<rot 90 0 0>}]
 dyn_hairFront_04_r [{shapeType=Capsule, length=5.8144, radius=2.3688, center=<1.8275 0.0002 0>, rotation=<rot 90 0 0>}]
 dyn_hairFront_Leaves [{shapeType=Capsule, length=4.613, radius=2.6598, center=<1.8364 0 0>, rotation=<rot 90 0 0>}]
 dyn_hairLeaf_A_l [{shapeType=Capsule, length=3.8785, radius=2.2957, center=<1.0404 0 0>, rotation=<rot 90 0 0>}]
 dyn_hairLeaf_B_l [{shapeType=Capsule, length=3.1456, radius=1.9862, center=<1.0404 0 0>, rotation=<rot 90 0 0>}]
 dyn_hairTie_l [{shapeType=Capsule, length=5.0906, radius=1.6302, center=<0.6843 0 0>, rotation=<rot 90 0 0>}]
 dyn_hairTie_r [{shapeType=Capsule, length=6.403, radius=1.3491, center=<1.0712 0.0003 0>, rotation=<rot 90 0 0>}]
 dyn_hairTopA_01 [{shapeType=Capsule, length=6.7836, radius=2.5861, center=<0.5966 0.0001 0.0003>, rotation=<rot 0 90 90>}]
 dyn_hairTop_B [{shapeType=Capsule, length=1.7198, radius=0.8783, rotation=<rot 90 0 0>}]
 dyn_hairTop_C [{shapeType=Capsule, length=5.8701, radius=2.0108, center=<0.4918 0.0001 0>, rotation=<rot 90 0 0>}]
 head [{shapeType=Capsule, length=17.8152, radius=4.3791, center=<-2.7852 -2.4752 0.0002>, rotation=<rot 90 0 -11.0925>}]
 spine_03 [{shapeType=Capsule, length=35, radius=15, shapeName=spine_03_capsule, rotation=<rot 0 -90 0>}]
 spine_04 [
  {shapeType=Capsule, length=92.25, radius=15, shapeName=spine_04_capsule, center=<18.4571 10.0473 -0.0001>}
  {shapeType=Box, extentX=48.7058, extentY=21.541, extentZ=76.6796, center=<11.6394 5.2989 -0.0001>, rotation=<rot 0 24.1535 0>}
 ]
 spine_05 [{shapeType=Capsule, length=105.3449, radius=10.0939, center=<0.5181 -0.7755 5.2904>}]
}
constraints [
 {bone1Name=dyn_hairFront_01_r, bone2Name=dyn_hairFront_01_Base_r, swing1LimitDegrees=40, swing2LimitDegrees=12, twistLimitDegrees=8}
 {bone1Name=dyn_hairFront_02_r, bone2Name=dyn_hairFront_01_r, swing1LimitDegrees=24, swing2LimitDegrees=12, twistLimitDegrees=8}
 {bone1Name=dyn_hairFront_03_r, bone2Name=dyn_hairFront_02_r, swing1LimitDegrees=24, swing2LimitDegrees=12, twistLimitDegrees=8}
 {bone1Name=dyn_hairFront_04_r, bone2Name=dyn_hairFront_03_r, swing1LimitDegrees=24, swing2LimitDegrees=12, twistLimitDegrees=8}
 {
  bone1Name=dyn_hairFront_01_Base_l
  bone2Name=head
  swing1LimitDegrees=45
  swing1Motion=Free
  swing2LimitDegrees=45
  swing2Motion=Free
  twistLimitDegrees=45
  twistMotion=Free
 }
 {bone1Name=dyn_hairFront_01_l, bone2Name=dyn_hairFront_01_Base_l, swing1LimitDegrees=20, swing2LimitDegrees=12, twistLimitDegrees=8}
 {bone1Name=dyn_hairFront_02_l, bone2Name=dyn_hairFront_01_l, swing1LimitDegrees=20, swing2LimitDegrees=12, twistLimitDegrees=4}
 {bone1Name=dyn_hairFront_03_l, bone2Name=dyn_hairFront_02_l, swing1LimitDegrees=20, swing2LimitDegrees=12, twistLimitDegrees=4}
 {
  bone1Name=head
  bone2Name=spine_05
  swing1LimitDegrees=45
  swing1Motion=Free
  swing2LimitDegrees=45
  swing2Motion=Free
  twistLimitDegrees=45
  twistMotion=Free
 }
 {
  bone1Name=spine_05
  bone2Name=spine_04
  swing1LimitDegrees=45
  swing1Motion=Free
  swing2LimitDegrees=45
  swing2Motion=Free
  twistLimitDegrees=45
  twistMotion=Free
 }
 {bone1Name=dyn_hairBack_01_c, bone2Name=head, swing1LimitDegrees=16, swing2LimitDegrees=10, twistLimitDegrees=6, twistMotion=Locked}
 {bone1Name=dyn_hairBack_02_c, bone2Name=dyn_hairBack_01_c, swing1LimitDegrees=16, swing2LimitDegrees=10, twistLimitDegrees=6}
 {bone1Name=dyn_hairBack_03_c, bone2Name=dyn_hairBack_02_c, swing1LimitDegrees=16, swing2LimitDegrees=10, twistLimitDegrees=6}
 {bone1Name=dyn_hairBack_04_c, bone2Name=dyn_hairBack_03_c, swing1LimitDegrees=16, swing2LimitDegrees=12, twistLimitDegrees=6}
 {bone1Name=dyn_hairBack_05_c, bone2Name=dyn_hairBack_04_c, swing1LimitDegrees=16, swing2LimitDegrees=12, twistLimitDegrees=6}
 {bone1Name=dyn_hairBackA_01, bone2Name=dyn_hairBack_03_c, swing1LimitDegrees=22, swing2LimitDegrees=12, twistLimitDegrees=6}
 {bone1Name=dyn_hairBackA_02, bone2Name=dyn_hairBackA_01, swing1LimitDegrees=16, swing2LimitDegrees=20, twistLimitDegrees=6}
 {
  bone1Name=dyn_hairBack_02_r
  bone2Name=dyn_hairBack_01_r
  swing1LimitDegrees=10
  swing2LimitDegrees=0
  swing2Motion=Locked
  twistLimitDegrees=6
  twistMotion=Locked
 }
 {bone1Name=dyn_hairBack_03_r, bone2Name=dyn_hairBack_02_r, swing1LimitDegrees=10, swing2LimitDegrees=0, twistLimitDegrees=6}
 {bone1Name=dyn_hairBack_02_l, bone2Name=dyn_hairBack_01_l, swing1LimitDegrees=22, swing2LimitDegrees=12, twistLimitDegrees=8}
 {bone1Name=dyn_hairBack_03_l, bone2Name=dyn_hairBack_02_l, swing1LimitDegrees=22, swing2LimitDegrees=12, twistLimitDegrees=8}
 {bone1Name=dyn_hairBack_04_l, bone2Name=dyn_hairBack_03_l, swing1LimitDegrees=22, swing2LimitDegrees=12, twistLimitDegrees=8}
 {bone1Name=dyn_hairTopA_01, bone2Name=head, swing1LimitDegrees=16, swing2LimitDegrees=10, twistLimitDegrees=6}
 {bone1Name=dyn_hairFront_Leaves, bone2Name=head, swing1LimitDegrees=20, swing2LimitDegrees=12, twistLimitDegrees=8}
 {bone1Name=dyn_hairLeaf_B_l, bone2Name=head, swing1LimitDegrees=20, swing2LimitDegrees=12, twistLimitDegrees=8}
 {bone1Name=dyn_hairLeaf_A_l, bone2Name=head, swing1LimitDegrees=20, swing2LimitDegrees=12, twistLimitDegrees=8}
 {bone1Name=dyn_hairTie_l, bone2Name=head, swing1LimitDegrees=20, swing2LimitDegrees=20, twistLimitDegrees=10}
 {bone1Name=dyn_hairTie_r, bone2Name=head, swing1LimitDegrees=12, swing2LimitDegrees=20, twistLimitDegrees=10}
 {
  bone1Name=dyn_hairTop_B
  bone2Name=head
  swing1LimitDegrees=45
  swing1Motion=Free
  swing2LimitDegrees=45
  swing2Motion=Free
  twistLimitDegrees=45
  twistMotion=Free
 }
 {bone1Name=dyn_hairTop_C, bone2Name=head, swing1LimitDegrees=16, swing2LimitDegrees=16, twistLimitDegrees=8}
 {
  bone1Name=dyn_hairBack_01_r
  bone2Name=dyn_hairBack_01_c
  swing1LimitDegrees=10
  swing2LimitDegrees=10
  twistLimitDegrees=6
  twistMotion=Locked
 }
 {bone1Name=dyn_hairBack_01_l, bone2Name=dyn_hairBack_01_c, swing1LimitDegrees=32, swing2LimitDegrees=12, twistLimitDegrees=6}
 {
  bone1Name=spine_04
  bone2Name=spine_03
  swing1LimitDegrees=45
  swing1Motion=Free
  swing2LimitDegrees=45
  swing2Motion=Free
  twistLimitDegrees=45
  twistMotion=Free
 }
 {
  bone1Name=dyn_hairFront_01_Base_r
  bone2Name=spine_03
  swing1LimitDegrees=45
  swing1Motion=Free
  swing2LimitDegrees=45
  swing2Motion=Free
  twistLimitDegrees=45
  twistMotion=Free
 }
]
