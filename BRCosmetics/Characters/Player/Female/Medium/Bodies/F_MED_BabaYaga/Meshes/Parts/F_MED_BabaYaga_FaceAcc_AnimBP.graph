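# Post-process style AnimBlueprint for the Baba Yaga face accessory skeleton, adding dynamic hair motion on top of the pose copied from the attached parent mesh.
# Each update maps neck_01 yaw delta relative to spine_05 to an alpha that additively offsets two hair bones, front-right and top.
# A RigidBody node then simulates physics in spine_05 space with scaled component motion and overridden gravity.
# A second RigidBody node is left unconnected.
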AnimBlueprint /BRCosmetics/Characters/Player/Female/Medium/Bodies/F_MED_BabaYaga/Meshes/Parts/F_MED_BabaYaga_FaceAcc_AnimBP
target_skeleton=&/F_MED_BabaYaga_FaceAcc_Skeleton
graphs [
 AnimGraph {nodes [
   RigidBody() {
    bForceDisableCollisionBetweenConstraintBodies=true
    bOverrideWorldGravity=true
    overridePhysicsAsset=&/F_MED_BabaYaga_Physics
    simulationSpace=BaseBoneSpace
    alphaBoolBlend {blendOption=Linear}
    baseBoneRef {boneName=spine_05}
    componentAppliedLinearAccClamp=<2000 2000 2000>
    componentLinearAccScale=<0.25 0.25 0.35>
    componentLinearVelScale=<0.2 0.25 0.25>
    overrideWorldGravity {z=-980}
   }
   n1 = Get hairTranslateAlpha() -> (hairTranslateAlpha:double)
   n2 = "Transform (Modify) Bone - Bone: dyn_hairFront_01_Base_r"(
    ComponentPose=Local To Component(LocalPose=Copy Pose From Mesh() {bUseAttachedParent=true})
    Alpha=@n1
    Translation="0,-2,2"
    Rotation="0,0,0"
    Scale="1,1,1"
   ) {translationMode=BMM_Additive, alphaBoolBlend {blendOption=Linear}, boneToModify {boneName=dyn_hairFront_01_Base_r}, translation {y=-2, z=2}}
   n3 = "Transform (Modify) Bone - Bone: dyn_hairTopA_01"(ComponentPose=@n2, Alpha=@n1, Translation="0,-1,1", Rotation="0,0,0", Scale="1,1,1") {translationMode=BMM_Additive, alphaBoolBlend {blendOption=Linear}, boneToModify {boneName=dyn_hairTopA_01}, translation {y=-1, z=1}}
   n4 = RigidBody(ComponentPose=@n3) {
    bForceDisableCollisionBetweenConstraintBodies=true
    bOverrideWorldGravity=true
    overridePhysicsAsset=&F_MED_BabaYaga_Physics
    simulationSpace=BaseBoneSpace
    alphaBoolBlend {blendOption=Linear}
    baseBoneRef {boneName=spine_05}
    componentAppliedLinearAccClamp=<2000 2000 2000>
    componentLinearAccScale=<0.25 0.35 0.35>
    componentLinearVelScale=<0.2 0.25 0.25>
    overrideWorldGravity {z=-980}
   }
   Output Pose(Result=Component To Local(ComponentPose=@n4)) {layerGroup=None}
  ]}
 EventGraph {nodes [
   Event BlueprintUpdateAnimation()
   n1 = BreakTransform(InTransform=GetDeltaTransformFromRefPose(self=GetOwnerSkeletalMeshComponent(), BoneName=neck_01, BaseName=spine_05))
   > Set hairTranslateAlpha(hairTranslateAlpha:double=MapRangeClamped(Value=BreakRotator(InRot=@n1.Rotation).Yaw, InRangeA=0, InRangeB=-20)) -> (Output_Get:double)
  ]}
]
variables [
 hairTranslateAlpha {type=number}
 hairTranslateAlphaCrouch {type=number}
 Translation {type=Vector, default=<0 5 0>}
 Translation_0 {type=Vector, default=<0 -1 1>}
 Translation_1 {type=Vector, default=<0 -1 0>}
 Translation_2 {type=Vector, default=<0 2 1>}
]
